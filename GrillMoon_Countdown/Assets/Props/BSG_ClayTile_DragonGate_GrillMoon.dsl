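; GrillMoon snowy Dragon Gate portal that thaws day by day through the countdown event.
; The server sets replicated Day from GrillMoon calendar events, with sequencer event-router listening as an editor fallback, and enables knockback collision.
; OnRep and relevancy refreshes call UpdatePortalPhase, which drives the Portal_Phase MPC for VFX, icicle and face visibility, and snow-melt material parameters.
; Editor construction zeroes the MPC so the portal does not glow.
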(event EventBeginPlay
  ; GrillMoon Snowy Portal
  (|CallParentFunction)
  (Class|FortAsyncActionCurrentPlaylistReady|PlaylistDataReadyAsync
    (:OnReady
      (if (Networking|IsServer)
        (CallFunction|WatchForDayEventsFromSequencer) ; watch for sequencer day events as a fallback if\nplaying in editor and the calendar isn't active
        (MCPContext|InitialCalendarSyncCompleted
          (:OnSyncComplete
            (if (Editor|InEditororPIE)
              (bind _biseventactive (CalendarEvents|GetCalendarEventInformation "GrillMoon_MM_Meditates"))
              (if (not _biseventactive)
                (CallFunction|WatchForDayEventsFromSequencer) ; watch for sequencer day events as a fallback if\nplaying in editor and the calendar isn't active
              ))
            (Utilities|FlowControl|Sequence
              (:then_0
                (bind _biseventactive_2 (CalendarEvents|GetCalendarEventInformation "GrillMoon_Day0"))
                (if _biseventactive_2
                  (CallFunction|SetDay "NewEnumerator30")))
              (:then_1
                (bind _biseventactive_3 (CalendarEvents|GetCalendarEventInformation "GrillMoon_Day1"))
                (if _biseventactive_3
                  (CallFunction|SetDay "NewEnumerator17")))
              (:then_2
                (bind _biseventactive_4 (CalendarEvents|GetCalendarEventInformation "GrillMoon_Day2"))
                (if _biseventactive_4
                  (CallFunction|SetDay "NewEnumerator18")))
              (:then_3
                (bind _biseventactive_5 (CalendarEvents|GetCalendarEventInformation "GrillMoon_Day3"))
                (if _biseventactive_5
                  (CallFunction|SetDay "NewEnumerator19")))
              (:then_4
                (bind _biseventactive_6 (CalendarEvents|GetCalendarEventInformation "GrillMoon_Day4"))
                (if _biseventactive_6
                  (CallFunction|SetDay "NewEnumerator20")))
              (:then_5
                (bind _biseventactive_7 (CalendarEvents|GetCalendarEventInformation "GrillMoon_Day5"))
                (if _biseventactive_7
                  (CallFunction|SetDay "NewEnumerator21")))))) ; server, check calendar window
        (else
          (Default|UpdatePortalPhase (Variables|Default|GetDay%20%28Replicated%29)) ; client when the portal comes into relevancy
        )))))

(event Custom|Watch%20for%20Day%20Events%20from%20Sequencer
  ; GrillMoon Snowy Portal
  (Utilities|FlowControl|DoOnce false
    (:Completed
      (bind (_asynctaskproxy _payload) (GameplayEventRouter|StartListeningToStatefulEvent (GameplayEventRouter|GetGlobalEventRouter) "/GrillMoon_Countdown/Gameplay/EventRouter/GrillMoon_Day_Event.GrillMoon_Day_Event")
        (:OnEventReceived
          _)
        (:OnSavedEventStateExists
          _))
      (CallFunction|SetDay (Utilities|Struct|BreakGrillMoonDayEvent _payload)))))

(event Custom|Set%20Day
  (params (Selection :type GrillMoon_Day_Event_Enum%20Enum :default "NewEnumerator30"))
  ; GrillMoon Snowy Portal
  (switch Utilities|FlowControl|Switch|SwitchonGrillMoon_Day_Event_Enum Selection
    (:NewEnumerator30
      (Variables|Default|SetDay%20%28Replicated%29 1))
    (:NewEnumerator17
      (Variables|Default|SetDay%20%28Replicated%29 2))
    (:NewEnumerator18
      (Variables|Default|SetDay%20%28Replicated%29 3))
    (:NewEnumerator19
      (Variables|Default|SetDay%20%28Replicated%29 4))
    (:NewEnumerator20
      (Variables|Default|SetDay%20%28Replicated%29 5))
    (:NewEnumerator21
      (Variables|Default|SetDay%20%28Replicated%29 5))
    (:NewEnumerator16
      (Variables|Default|SetDay%20%28Replicated%29 6)))
  (CallFunction|PortalCollisionEnabled true))

(event Custom|Portal%20Collision%20Enabled
  (params (Collision%20Enabled :type Boolean))
  (bind _impulsecollision (Variables|Default|GetImpulseCollision))
  (bind _box___collision (Variables|Default|GetBox%20-%20Collision))
  ; Portal Collision & knock Back
  (Variables|/Script/Engine.PrimitiveComponent|SetbGenerateOverlapEvents Collision%20Enabled _impulsecollision) ; works on the server
  (Components|Sphere|SetSphereRadius _impulsecollision 2e+02 false)
  (if Collision%20Enabled
    (Collision|SetCollisionResponsetoChannel _box___collision "ECC_WorldDynamic" "ECR_Block")
    (else
      (Collision|SetCollisionResponsetoChannel _box___collision "ECC_WorldDynamic" "ECR_Ignore"))))

(fn ChangePortalState
  (params (Active :type Boolean))
  (|CallParentFunction Active)
  (Rendering|SetVisibility (Variables|Default|GetNE_SystemShare) Active false))

(fn OnRep_Day(Replicated)
  (Default|UpdatePortalPhase (Variables|Default|GetDay%20%28Replicated%29)))

(fn Portal-ResetToStart
  (Default|UpdatePortalPhase 0))

(fn PortalDay01
  (Default|UpdatePortalPhase 1))

(fn PortalDay02
  (Default|UpdatePortalPhase 2))

(fn PortalDay03
  (Default|UpdatePortalPhase 3))

(fn PortalDay04
  (Default|UpdatePortalPhase 4))

(fn PortalDay05
  (Default|UpdatePortalPhase 5))

(fn SetIcicleStage
  (params (Stage%20%281-4%29 :type Integer))
  (switch int Stage%20%281-4%29
    (:0
      _)
    (:1
      _)
    (:2
      (Rendering|SetVisibility (Variables|Default|GetSM_FrostA) true false) ; Show Medium Ice
      (Rendering|SetVisibility (Variables|Default|GetSM_SnowC%20-%20Small) false false) ; Hide Small, Large Ice
    )
    (:3
      (Rendering|SetVisibility (Variables|Default|GetSM_FrostA) true false) ; Show Small Ice
      (Rendering|SetVisibility (Variables|Default|GetSM_SnowA%20-%20Biggest) false false) ; Hide Medium, Large Ice
    )
    (:4
      (Rendering|SetVisibility (Variables|Default|GetSM_SnowA%20-%20Biggest) false false) ; Hide all
    ))
  (Rendering|SetVisibility (Variables|Default|GetSM_SnowA%20-%20Biggest) true false) ; Show Big Ice, Frost
  (Rendering|SetVisibility (Variables|Default|GetSM_SnowB%20-%20Medium) false false) ; Hide Small, Medium Ice
)

(fn SetMID-Mask_WSNoise
  (params (Value :type Float%20%28single-precision%29)))

(fn SetMID-SRM_Bias
  (params (Value :type Float%20%28single-precision%29))
  (Rendering|Material|SetScalarParameterValue (Class|BTechArtGlobalLibrary|GetDynamicMaterialInstance (Variables|StaticMeshActor|GetStaticMeshComponent self)) "SRM_Bias (Contract/Expand)" Value :!class /Script/Engine.MaterialInstanceDynamic))

(fn SetMID-Top_Bias
  (params (Value :type Float%20%28single-precision%29))
  (Rendering|Material|SetScalarParameterValue (Class|BTechArtGlobalLibrary|GetDynamicMaterialInstance (Variables|StaticMeshActor|GetStaticMeshComponent self)) "Mask_Gradient Top_Bias (Contract/Expand)" Value :!class /Script/Engine.MaterialInstanceDynamic))

(fn SetMPCforPortalPhase
  (params (Day :type Integer))
  (Rendering|Material|SetScalarParameterValue (Variables|Default|GetMPC_GrillMoon) (Utilities|Name|MakeLiteralName "Portal_Phase") (Utilities|Select :Index Day :Option%200 0.0 :Option%201 1.0 :Option%202 2.0 :Option%203 3.0 :Option%204 4.0 :Option%205 4.0 :Option%206 4.0) :!class /Script/Engine.KismetMaterialLibrary) ; how visible the VFX is in the portal day to day
)

(fn SetPortalFaceVisibility
  (params (bFaceVisible :type Boolean))
  (Rendering|SetVisibility (Variables|Default|GetSM_Head self) bFaceVisible false) ; Show/Hide the face in the portal.\n\nConstruct on the portal also does this. Wouldn't normally\nduplicate the logic here, but hoping to avoid making\nchanges to the initialization of the portal blueprint.
)

(fn UpdatePortalPhase
  (params (Day :type Integer))
  (Default|SetMPCForPortalPhase (Variables|GetDay)) ; VFX in the portal
  (Utilities|FlowControl|Sequence
    (:then_0
      ;[0.3323 0.278 0.6094 0.907] Snow Meshes
      (switch int (Variables|GetDay)
        (:Default
          _)
        (:0
          _)
        (:1
          _)
        (:2
          (Default|SetIcicleStage 2) ; 2 - middle sized icicle
        )
        (:3
          (Default|SetIcicleStage 3) ; 3 - smallest icicle
        )
        (:4
          (Default|SetIcicleStage 4 :!label _label) ; 4 - no icicles
        )
        (:5
          (goto _label))
        (:6
          (goto _label)))
      (Default|SetIcicleStage 1) ; 1 - biggest sized icicle
    )
    (:then_1
      ;[0.3323 0.278 0.6094 0.907] Show / Hide Face
      (switch int (Variables|GetDay)
        (:Default
          _)
        (:0
          _)
        (:1
          _)
        (:2
          _)
        (:3
          _)
        (:4
          (Default|SetPortalFaceVisibility true :!label _label_2) ; show the face
        )
        (:5
          (goto _label_2))
        (:6
          _))
      (Default|SetPortalFaceVisibility false) ; hide the face
    )
    (:then_2
      ;[0.3323 0.278 0.6094 0.907] Snow Material MID / Snow Melting
      (switch int (Variables|GetDay)
        (:Default
          _)
        (:0
          _)
        (:1
          (Default|SetMID-TopBias (Math|Float|MapRangeClamped (Math|Conversions|ToFloat(Integer) (Variables|GetDay)) 0.0 6.0 1.5 0.4) :!label _label_3)
          (Default|SetMID-SRMBias (Math|Float|MapRangeClamped (Math|Conversions|ToFloat(Integer) (Variables|GetDay)) 0.0 6.0 -0.5 -0.2)))
        (:2
          (goto _label_3))
        (:3
          (goto _label_3))
        (:4
          (goto _label_3))
        (:5
          (Default|SetMID-TopBias 0.0 :!label _label_4)
          (Default|SetMID-SRMBias 0.0))
        (:6
          (goto _label_4)))
      (Default|SetMID-TopBias 1.5)
      (Default|SetMID-SRMBias -0.5))))

(fn ConstructionScript
  (|CallParentFunction)
  (if (Editor|InEditor)
    (Default|SetMPCForPortalPhase 0) ; Just in editor, set the MPC phase to be empty so\nthe portal isn't glowing by default.
  ))
